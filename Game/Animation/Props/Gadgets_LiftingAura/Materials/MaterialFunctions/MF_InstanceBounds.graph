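# MF_InstanceBounds is a material function in the LiftingAura gadget folder that exposes an instance's local-space bounds.
# Two Custom HLSL nodes read InstanceLocalBoundsCenter and InstanceLocalBoundsExtent directly from the primitive data.
# It outputs the bounds minimum as center minus extent, the maximum as center plus extent, and the size as twice the extent.
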MaterialFunction /Game/Animation/Props/Gadgets_LiftingAura/Materials/MaterialFunctions/MF_InstanceBounds
graph {nodes [
  n1 = Custom() {code="GetPrimitiveData(Parameters).InstanceLocalBoundsCenter.xyz", description=Instance Bounds Center, inputs []}
  n2 = Custom() {code="GetPrimitiveData(Parameters).InstanceLocalBoundsExtent.xyz", description=Instance Bounds Extent, inputs []}
  FunctionOutput(None=Subtract(A=@n1, B=@n2)) {outputName=Instance Bounds Minimum}
  FunctionOutput(None=Add(A=@n1, B=@n2)) {outputName=Instance Bounds Max, sortPriority=1}
  FunctionOutput(None=Multiply(A=@n2) {constB=2}) {outputName=Instance Bounds Size, sortPriority=2}
 ]}
